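# M_SimpleDiffuse is a minimal Default Lit material for mesh particles, filed under Dusty world prop effects.
# Its graph holds only a Particle Color node and two 0.5 constants, with no textures or other math.
# The listing shows no links from these nodes to material outputs, though the name suggests particle color serves as the diffuse color.
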Material /Clyde/Effects/Fort_Effects/World/Props/Dusty/M_SimpleDiffuse
graph {
 properties {bUsedWithMeshParticles=true, pixelDepthOffsetMode=PDOM_Legacy, usedShadingModels=Default Lit}
 nodes [
  Constant() {r=0.5}
  Constant() {r=0.5}
  ParticleColor()
 ]
}
